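# Physics asset for the ToadLeaf pickaxe, giving its mesh a forked dangling rope for secondary motion.
# Capsule bodies cover the axe base, a small rope anchor and the rope chain, which runs from one center segment into left and right three-segment branches ending in round, sphere-like tips.
# The anchor swings and twists freely on the base, while downstream joints have angular limits that generally tighten toward the tips.
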PhysicsAsset /BRCosmetics/Weapons/FORT_Melee/Pickaxe_ToadLeaf/Meshes/ToadLeaf_Axe_Physics
bodies {
 Base [{
   shapeType=Capsule
   length=27.7204
   radius=8.8684
   shapeName=Base_capsule
   center=<0.0072 -0.0001 17.8807>
   rotation=<rot -0.0216 -90 -0.0688>
  }]
 Base_rope_C [{shapeType=Capsule, length=0.505, radius=0.505, shapeName=Base_rope_C_capsule}]
 Dyn_rope_C_01 [{shapeType=Capsule, length=1, radius=2, shapeName=Dyn_rope_C_01_capsule, center=<1.1821 0 0>, rotation=<rot -90 0.0007 -0.0003>}]
 Dyn_rope_L_01 [{shapeType=Capsule, length=2.5, radius=1.5, shapeName=Dyn_rope_L_01_capsule, center=<1.6775 0 -0.1241>, rotation=<rot 80 0.0047 0.0032>}]
 Dyn_rope_L_02 [{shapeType=Capsule, length=1.5, radius=1, shapeName=Dyn_rope_L_02_capsule, center=<1.0146 0 0>, rotation=<rot -90 -179.9471 179.9482>}]
 Dyn_rope_L_03 [{shapeType=Capsule, length=0, radius=2.5, shapeName=Dyn_rope_L_03_capsule, center=<2.79 0.0001 0>, rotation=<rot -90 -179.9956 179.9972>}]
 Dyn_rope_R_01 [{shapeType=Capsule, length=2, radius=1.5, shapeName=Dyn_rope_R_01_capsule, center=<1.7415 0 0.1354>, rotation=<rot -80 0.004 -0.0024>}]
 Dyn_rope_R_02 [{shapeType=Capsule, length=1.5, radius=1, shapeName=Dyn_rope_R_02_capsule, center=<1.0146 0 0>, rotation=<rot -90 166.7373 -166.7372>}]
 Dyn_rope_R_03 [{shapeType=Capsule, length=0, radius=2.5, shapeName=Dyn_rope_R_03_capsule, center=<2.79 0 0>, rotation=<rot -90 165.149 -165.1489>}]
}
constraints [
 {
  bone1Name=Base_rope_C
  bone2Name=Base
  swing1LimitDegrees=45
  swing1Motion=Free
  swing2LimitDegrees=45
  swing2Motion=Free
  twistLimitDegrees=45
  twistMotion=Free
 }
 {bone1Name=Dyn_rope_C_01, bone2Name=Base_rope_C, swing1LimitDegrees=40, swing2LimitDegrees=30, twistLimitDegrees=5}
 {bone1Name=Dyn_rope_L_01, bone2Name=Dyn_rope_C_01, swing1LimitDegrees=40, swing2LimitDegrees=20, twistLimitDegrees=15}
 {bone1Name=Dyn_rope_L_02, bone2Name=Dyn_rope_L_01, swing1LimitDegrees=20, swing2LimitDegrees=20, twistLimitDegrees=15}
 {bone1Name=Dyn_rope_L_03, bone2Name=Dyn_rope_L_02, swing1LimitDegrees=10, swing2LimitDegrees=15, twistLimitDegrees=5}
 {bone1Name=Dyn_rope_R_01, bone2Name=Dyn_rope_C_01, swing1LimitDegrees=35, swing2LimitDegrees=25, twistLimitDegrees=15}
 {bone1Name=Dyn_rope_R_02, bone2Name=Dyn_rope_R_01, swing1LimitDegrees=20, swing2LimitDegrees=30, twistLimitDegrees=15}
 {bone1Name=Dyn_rope_R_03, bone2Name=Dyn_rope_R_02, swing1LimitDegrees=10, swing2LimitDegrees=15, twistLimitDegrees=5}
]
